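# Conversation database for the Lifeguard specialist NPC, providing a surfer-slang greeting and routing into shared conversation branches.
# The entry point leads to one speech node, whose default line has prioritized contextual alternatives for a first-ever conversation or specific player outfits.
# The speech then branches through link nodes to the remote NPC Library quest, basic services and Ch5S1.D3.Q8 conversation entries.
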ConversationDatabase /NPCLibrary/NPCs/Lifeguard/NPCLibrary_Conversation_Lifeguard
entry_points [{tag=Conversation.Entry.Specialist.Lifeguard, destinations [#1]}]
nodes [
 ConversationEntryPointNode_7 {
  id=#1
  class=ConversationEntryPointNode
  outputs [#2]
  properties {entryTag {tagName=Conversation.Entry.Specialist.Lifeguard}}
 }
 Speech_C_0 {
  id=#2
  class=Speech_C
  outputs [#3, #4, #5]
  properties {
   nodeName=""
   generalConfig {
    defaultMessage="Lookin' to charge some heavy bombs, bro?"
    contextualMessages [
     {
      message="I thought I caught my last wave, bro. But if you need my help..."
      contextRequirements [{requirement {class=FortControllerRequirement_IsFirstEverConversationWithNPC}}]
      priority {curve {rowName=Tandem.Speech.Priority.FirstEverInteraction}}
     }
     {
      message="I know, I know, I should put down the board. But have you seen the bombs out there, bro?"
      contextRequirements [{requirement {
         class=FortControllerRequirement_HasCID
         properties {allowedCIds [/BRCosmetics/Athena/Items/Cosmetics/Characters/CID_166_Athena_Commando_F_Lifeguard]}
        }}]
      priority {curve {rowName=Tandem.Speech.Priority.OutfitReaction}}
     }
     {
      message=I was told you know some radical tricks. Ever snap off-the-lip?
      contextRequirements [{requirement {class=FortControllerRequirement_HasCID, properties {allowedCIds [/BRCosmetics/Athena/Items/Cosmetics/Characters/Character_SparkArcher]}}}]
      priority {curve {rowName=Tandem.Speech.Priority.OutfitReaction}}
     }
     {
      message="Woah, bro. Something tells me you don't like water."
      contextRequirements [{requirement {
         class=FortControllerRequirement_HasCID
         properties {allowedCIds [
           /BRCosmetics/Athena/Items/Cosmetics/Characters/Character_BuffCatCruise
           &Character_BengalBasher_Suit
          ]}
        }}]
      priority {curve {rowName=Tandem.Speech.Priority.OutfitReaction}}
     }
    ]
   }
  }
 }
 ConversationLinkNode_2 {id=#5, class=ConversationLinkNode, properties {remoteEntryTag {tagName=Conversation.Entry.Ch5S1.D3.Q8}}}
 ConversationLinkNode_0 {id=#4, class=ConversationLinkNode, properties {remoteEntryTag {tagName=Conversation.Entry.BasicServices}}}
 ConversationLinkNode_1 {id=#3, class=ConversationLinkNode, properties {remoteEntryTag {tagName=Conversation.Entry.NPCLibrary.Quests}}}
]
